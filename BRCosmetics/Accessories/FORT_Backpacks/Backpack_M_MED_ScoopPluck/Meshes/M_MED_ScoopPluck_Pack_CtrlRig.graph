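# Control Rig for the ScoopPluck backpack cosmetic, animating a small figure with pelvis, neck, head, tie, arms and legs.
# On setup, heel, toe, ball and foot reverse-foot controls and nulls take initial transforms from leg bones and offset reference nulls.
# Forward solve copies FK controls onto body and arm bones, then runs two-bone leg IK to the reverse-foot nulls.
# Calf controls supply pole vectors (Direction mode), and rotation constraints drive the ball bones.
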ControlRigBlueprint /BRCosmetics/Accessories/FORT_Backpacks/Backpack_M_MED_ScoopPluck/Meshes/M_MED_ScoopPluck_Pack_CtrlRig
graphs [
 New Function_ContainedGraph {nodes [
   RigVMFunctionEntryNode()
   > &RigUnit_SetTransform(
    Item=(Type=Bone,Name="bp_pelvis")
    Weight=1
    bInitial=False
    Value=&RigUnit_GetTransform(Item=(Type=Control,Name="bp_pelvis_ctrl"), bInitial=False).Transform
   )
   RigUnit_GetTransform = &RigUnit_GetTransform(Item=(Type=Control,Name="bp_neck_ctrl"), bInitial=False)
   > RigUnit_SetTransform = &RigUnit_SetTransform(Item=(Type=Bone,Name="bp_neck"), Weight=1, bInitial=False, Value=@RigUnit_GetTransform.Transform)
   > &RigUnit_SetTransform(
    Item=(Type=Bone,Name="bp_head")
    Weight=1
    bInitial=False
    Value=&RigUnit_GetTransform(Item=(Type=Control,Name="bp_Newhead_ctrl"), bInitial=False).Transform
   )
   > &RigUnit_SetTransform(
    Item=(Type=Bone,Name="bp_tie")
    Weight=1
    bInitial=False
    Value=&RigUnit_GetTransform(Item=(Type=Control,Name="bp_tie_ctrl"), bInitial=False).Transform
   )
   > &RigUnit_SetTransform(
    Item=(Type=Bone,Name="FK_armRot_l")
    Weight=1
    bInitial=False
    Value=&RigUnit_GetTransform(Item=(Type=Control,Name="FK_armRot_l_ctrl"), bInitial=False).Transform
   )
   > &RigUnit_SetTransform(
    Item=(Type=Bone,Name="FK_upperarm_l")
    Weight=1
    bInitial=False
    Value=&RigUnit_GetTransform(Item=(Type=Control,Name="FK_upperarm_l_ctrl"), bInitial=False).Transform
   )
   > &RigUnit_SetTransform(
    Item=(Type=Bone,Name="FK_lowerarm_l")
    Weight=1
    bInitial=False
    Value=&RigUnit_GetTransform(Item=(Type=Control,Name="FK_lowerarm_l_ctrl"), bInitial=False).Transform
   )
   > &RigUnit_SetTransform(
    Item=(Type=Bone,Name="FK_hand_l")
    Weight=1
    bInitial=False
    Value=&RigUnit_GetTransform(Item=(Type=Control,Name="FK_hand_l_ctrl"), bInitial=False).Transform
   )
   > &RigUnit_SetTransform(
    Item=(Type=Bone,Name="FK_armRot_r")
    Weight=1
    bInitial=False
    Value=&RigUnit_GetTransform(Item=(Type=Control,Name="FK_armRot_r_ctrl"), bInitial=False).Transform
   )
   > &RigUnit_SetTransform(
    Item=(Type=Bone,Name="FK_upperarm_r")
    Weight=1
    bInitial=False
    Value=&RigUnit_GetTransform(Item=(Type=Control,Name="FK_upperarm_r_ctrl"), bInitial=False).Transform
   )
   > &RigUnit_SetTransform(
    Item=(Type=Bone,Name="FK_lowerarm_r")
    Weight=1
    bInitial=False
    Value=&RigUnit_GetTransform(Item=(Type=Control,Name="FK_lowerarm_r_ctrl"), bInitial=False).Transform
   )
   > &RigUnit_SetTransform(
    Item=(Type=Bone,Name="FK_hand_r")
    Weight=1
    bInitial=False
    Value=&RigUnit_GetTransform(Item=(Type=Control,Name="FK_hand_r_ctrl"), bInitial=False).Transform
   )
   > RigVMFunctionReturnNode()
  ]}
 New Function_ContainedGraph {nodes [
   RigVMFunctionEntryNode()
   > &RigUnit_SetTransform(
    Item=(Type=Null,Name="IK_calf_r_null_ref")
    Value=(Rotation=(X=0.4208,Y=0.5806,Z=-0.3169,W=0.6209),Translation=(X=7.5238,Y=-13.8977,Z=-12.4461),Scale3D=(X=1,Y=1,Z=1))
    Weight=1
    bInitial=False
   )
   > &RigUnit_SetTransform(
    Item=(Type=Null,Name="IK_calf_l_null_ref")
    Value=(Rotation=(X=0.5806,Y=0.4208,Z=-0.6209,W=0.3169),Translation=(X=-7.5238,Y=-13.8977,Z=-12.4461),Scale3D=(X=1,Y=1,Z=1))
    Weight=1
    bInitial=False
   )
   RigUnit_GetTransform = &RigUnit_GetTransform(Item=(Type=Null,Name="IK_Foot_RV_l"), bInitial=False)
   > /Script/ControlRig.RigUnit_TwoBoneIKSimplePerItem(
    DebugSettings=(bEnabled=False,Scale=10,WorldOffset=(Rotation=(X=0,Y=0,Z=0,W=1),Translation=(X=0,Y=0,Z=0),Scale3D=(X=1,Y=1,Z=1)))
    EffectorItem=(Type=Bone,Name="IK_foot_l")
    ItemA=(Type=Bone,Name="IK_thigh_l")
    ItemALength=0
    ItemB=(Type=Bone,Name="IK_calf_l")
    ItemBLength=0
    PoleVectorKind=Direction
    PoleVectorSpace=(Type=Bone,Name="None")
    PrimaryAxis=(X=1,Y=0,Z=0)
    SecondaryAxis=(X=0,Y=0,Z=1)
    SecondaryAxisWeight=1
    StretchMaximumRatio=1.25
    StretchStartRatio=0.75
    Weight=1
    bEnableStretch=False
    bPropagateToChildren=True
    Effector=@RigUnit_GetTransform.Transform
    PoleVector=&RigUnit_GetTransform(Item=(Type=Control,Name="IK_calf_l_ctrl"), bInitial=false).Transform.Translation
   ) {is=struct RigVMUnitNode}
   > &RigUnit_TwoBoneIKSimplePerItem(
    DebugSettings=(bEnabled=False,Scale=10,WorldOffset=(Rotation=(X=0,Y=0,Z=0,W=1),Translation=(X=0,Y=0,Z=0),Scale3D=(X=1,Y=1,Z=1)))
    EffectorItem=(Type=Bone,Name="IK_foot_r")
    ItemA=(Type=Bone,Name="IK_thigh_r")
    ItemALength=0
    ItemB=(Type=Bone,Name="IK_calf_r")
    ItemBLength=0
    PoleVectorKind=Direction
    PoleVectorSpace=(Type=Bone,Name="None")
    PrimaryAxis=(X=1,Y=0,Z=0)
    SecondaryAxis=(X=0,Y=0,Z=1)
    SecondaryAxisWeight=1
    StretchMaximumRatio=1.25
    StretchStartRatio=0.75
    Weight=1
    bEnableStretch=False
    bPropagateToChildren=True
    Effector=&RigUnit_GetTransform(Item=(Type=Null,Name="IK_Foot_RV_r"), bInitial=False).Transform
    PoleVector=&RigUnit_GetTransform(Item=(Type=Control,Name="IK_calf_r_ctrl"), bInitial=False).Transform.Translation
   )
   > &RigUnit_RotationConstraintLocalSpaceOffset(
    AdvancedSettings=(InterpolationType=Average,RotationOrderForFilter=XZY)
    Child=(Type=Bone,Name="IK_ball_l")
    Filter=(bX=True,bY=True,bZ=True)
    Parents=((Item=(Type=Null,Name="IK_Ball_RV_l"),Weight=1))
    Weight=1
   )
   > &RigUnit_RotationConstraintLocalSpaceOffset(
    AdvancedSettings=(InterpolationType=Average,RotationOrderForFilter=XZY)
    Child=(Type=Bone,Name="IK_ball_r")
    Filter=(bX=True,bY=True,bZ=True)
    Parents=((Item=(Type=Null,Name="IK_Ball_RV_r"),Weight=1))
    Weight=1
   )
   > RigVMFunctionReturnNode()
  ]}
 New Function_ContainedGraph {nodes [
   RigVMFunctionEntryNode()
   > &RigUnit_SetTransform(
    Item=(Type=Null,Name="IK_heel_l_ref")
    Space=LocalSpace
    Value=(Rotation=(X=0,Y=0,Z=0,W=1),Translation=(X=0,Y=0,Z=-2.35),Scale3D=(X=1,Y=1,Z=1))
    Weight=1
    bInitial=False
   )
   > &RigUnit_SetTransform(
    Item=(Type=Null,Name="IK_heel_r_ref")
    Space=LocalSpace
    Value=(Rotation=(X=0,Y=0,Z=0,W=1),Translation=(X=0,Y=0,Z=-2.35),Scale3D=(X=1,Y=1,Z=1))
    Weight=1
    bInitial=False
   )
   > &RigUnit_SetTransform(
    Item=(Type=Null,Name="IK_toe_l_ref")
    Space=LocalSpace
    Value=(Rotation=(X=0,Y=0,Z=0,W=-1),Translation=(X=0.9938,Y=0,Z=-0.5531),Scale3D=(X=1,Y=1,Z=1))
    Weight=1
    bInitial=False
   )
   > &RigUnit_SetTransform(
    Item=(Type=Null,Name="IK_toe_r_ref")
    Space=LocalSpace
    Value=(Rotation=(X=0,Y=0,Z=0,W=1),Translation=(X=0.9938,Y=0,Z=-0.5531),Scale3D=(X=1,Y=1,Z=1))
    Weight=1
    bInitial=False
   )
   RigUnit_GetTransform = &RigUnit_GetTransform(Item=(Type=Null,Name="IK_heel_l_ref"), bInitial=False)
   > RigUnit_SetTransform = &RigUnit_SetTransform(Item=(Type=Control,Name="IK_Heel_RV_l_ctrl"), Weight=1, bInitial=true, Value=@RigUnit_GetTransform.Transform)
   > &RigUnit_SetTransform(
    Item=(Type=Control,Name="IK_Heel_RV_r_ctrl")
    Weight=1
    bInitial=true
    Value=&RigUnit_GetTransform(Item=(Type=Null,Name="IK_heel_r_ref"), bInitial=False).Transform
   )
   > &RigUnit_SetTransform(
    Item=(Type=Control,Name="IK_Toe_RV_l_ctrl")
    Weight=1
    bInitial=true
    Value=&RigUnit_GetTransform(Item=(Type=Null,Name="IK_toe_l_ref"), bInitial=False).Transform
   )
   > &RigUnit_SetTransform(
    Item=(Type=Control,Name="IK_Toe_RV_r_ctrl")
    Weight=1
    bInitial=true
    Value=&RigUnit_GetTransform(Item=(Type=Null,Name="IK_toe_r_ref"), bInitial=False).Transform
   )
   > &RigUnit_SetTransform(
    Item=(Type=Control,Name="IK_Ball_RV_l_ctrl")
    Weight=1
    bInitial=true
    Value=&RigUnit_GetTransform(Item=(Type=Bone,Name="IK_ball_l"), bInitial=False).Transform
   )
   > &RigUnit_SetTransform(
    Item=(Type=Control,Name="IK_Ball_RV_r_ctrl")
    Weight=1
    bInitial=true
    Value=&RigUnit_GetTransform(Item=(Type=Bone,Name="IK_ball_r"), bInitial=False).Transform
   )
   > &RigUnit_SetTransform(
    Item=(Type=Null,Name="IK_Foot_RV_l")
    Weight=1
    bInitial=true
    Value=&RigUnit_GetTransform(Item=(Type=Bone,Name="IK_foot_l"), bInitial=False).Transform
   )
   > &RigUnit_SetTransform(
    Item=(Type=Null,Name="IK_Foot_RV_r")
    Weight=1
    bInitial=true
    Value=&RigUnit_GetTransform(Item=(Type=Bone,Name="IK_foot_r"), bInitial=False).Transform
   )
   > &RigUnit_SetTransform(
    Item=(Type=Null,Name="IK_Ball_RV_l")
    Weight=1
    bInitial=true
    Value=&RigUnit_GetTransform(Item=(Type=Bone,Name="IK_ball_l"), bInitial=False).Transform
   )
   > &RigUnit_SetTransform(
    Item=(Type=Null,Name="IK_Ball_RV_r")
    Weight=1
    bInitial=true
    Value=&RigUnit_GetTransform(Item=(Type=Bone,Name="IK_ball_r"), bInitial=False).Transform
   )
   > RigVMFunctionReturnNode()
  ]}
 RigVMFunctionLibrary {nodes [
   FK_Controls_Setup = RigVMCollapseNode(ExecuteContext=())
   Foot_Controls_Setup = RigVMCollapseNode(ExecuteContext=())
   IK_Controls_Setup = RigVMCollapseNode(ExecuteContext=())
  ]}
 RigVMModel {nodes [
   n1 = &RigUnit_PrepareForExecution()
   Foot_Controls_Setup = RigVMFunctionReferenceNode(ExecuteContext=@n1.ExecutePin)
   RigUnit_BeginExecution = &RigUnit_BeginExecution()
   FK_Controls_Setup = RigVMFunctionReferenceNode(ExecuteContext=@RigUnit_BeginExecution.ExecutePin)
   > IK_Controls_Setup = RigVMFunctionReferenceNode()
  ]}
]
hierarchy [
 root {loc=<0 0 0>, rot=<0 0 0 1>}
 pack {parent=root, loc=<0 -0.7316 0>, rot=<0 0 0 1>}
 bp_pelvis {parent=pack, loc=<0 -3.5544 -5.8161>, rot=<0 0 0.7071 -0.7071>}
 IK_thigh_l {parent=bp_pelvis, loc=<0.216 -0.8914 -0.1163>, rot=<-0.0727 -0.6204 0.2529 -0.7388>}
 IK_calf_l {parent=IK_thigh_l, loc=<6.5913 0 0>, rot=<-0.0043 -0.1285 0.0004 -0.9917>}
 IK_foot_l {parent=IK_calf_l, loc=<5.5251 0 0>, rot=<0.0175 0.5233 -0.0536 -0.8503>}
 IK_ball_l {parent=IK_foot_l, loc=<3.0246 0.0055 -0.0102>, rot=<0.0186 0.2487 -0.0258 -0.9681>}
 IK_toe_l_ref {parent=IK_ball_l, type=Null, loc=<0 0 0>, rot=<0 0 0 -1>}
 IK_heel_l_ref {parent=IK_foot_l, type=Null, loc=<0 0 0>, rot=<0 0 0 -1>}
 IK_thigh_r {parent=bp_pelvis, loc=<0.216 0.8914 -0.1163>, rot=<0.0727 -0.6204 -0.2529 -0.7388>}
 IK_calf_r {parent=IK_thigh_r, loc=<6.5913 0 0>, rot=<0.0043 -0.1285 -0.0004 -0.9917>}
 IK_foot_r {parent=IK_calf_r, loc=<5.5251 0 0>, rot=<-0.0175 0.5233 0.0536 -0.8503>}
 IK_ball_r {parent=IK_foot_r, loc=<3.0246 -0.0056 -0.0102>, rot=<-0.0186 0.2487 0.0258 -0.9681>}
 IK_toe_r_ref {parent=IK_ball_r, type=Null, loc=<0 0 0>, rot=<0 0 0 1>}
 IK_heel_r_ref {parent=IK_foot_r, type=Null, loc=<0 0 0>, rot=<0 0 0 1>}
 bp_neck {parent=bp_pelvis, loc=<-0.6852 0 7.7908>, rot=<-0.5845 0.3959 -0.4001 -0.5843>}
 bp_head {parent=bp_neck, loc=<1.6889 0 0>, rot=<0.435 -0.5575 0.5575 -0.435>}
 FK_armRot_l {parent=bp_pelvis, loc=<-0.8773 -2.7324 6.7308>, rot=<-0.4999 -0.4861 0.4975 -0.516>}
 FK_upperarm_l {parent=FK_armRot_l, loc=<0 0 0.6595>, rot=<0 0.1614 0 -0.9869>}
 FK_lowerarm_l {parent=FK_upperarm_l, loc=<4.4091 0.0711 -0.0243>, rot=<0.1446 0.03 -0.1975 -0.9691>}
 FK_hand_l {parent=FK_lowerarm_l, loc=<3.4064 -0.0033 0.0933>, rot=<-0.1451 -0.211 -0.0106 -0.9666>}
 FK_armRot_r {parent=bp_pelvis, loc=<-0.8773 2.7324 6.7308>, rot=<0.4975 -0.516 0.4999 0.4861>}
 FK_upperarm_r {parent=FK_armRot_r, loc=<0 0 -0.6595>, rot=<0 0.1614 0 -0.9869>}
 FK_lowerarm_r {parent=FK_upperarm_r, loc=<-4.4091 -0.0711 0.0243>, rot=<0.1446 0.03 -0.1975 -0.9691>}
 FK_hand_r {parent=FK_lowerarm_r, loc=<-3.4064 0.0033 -0.0933>, rot=<-0.1451 -0.211 -0.0106 -0.9666>}
 bp_tie {parent=pack, loc=<0 -4.2361 -3.9506>, rot=<0 0 0.7071 -0.7071>}
 Main_ctrl {controlType=EULER_TRANSFORM, type=Control, loc=<0 0 0>, rot=<0 0 0 1>}
 bp_pelvis_ctrl {controlType=EULER_TRANSFORM, parent=Main_ctrl, type=Control, loc=<0 0 0>, rot=<0 0 0 1>}
 bp_neck_ctrl {controlType=EULER_TRANSFORM, parent=bp_pelvis_ctrl, type=Control, loc=<0 0 0>, rot=<0 0 0 1>}
 bp_Newhead_ctrl {controlType=EULER_TRANSFORM, parent=bp_neck_ctrl, type=Control, loc=<0 0 0>, rot=<0 0 0 1>}
 FK_armRot_l_ctrl {controlType=EULER_TRANSFORM, parent=bp_pelvis_ctrl, type=Control, loc=<0 0 0>, rot=<0 0 0 1>}
 FK_upperarm_l_ctrl {controlType=EULER_TRANSFORM, parent=FK_armRot_l_ctrl, type=Control, loc=<0 0 0>, rot=<0 0 0 1>}
 FK_lowerarm_l_ctrl {controlType=EULER_TRANSFORM, parent=FK_upperarm_l_ctrl, type=Control, loc=<0 0 0>, rot=<0 0 0 1>}
 FK_hand_l_ctrl {controlType=EULER_TRANSFORM, parent=FK_lowerarm_l_ctrl, type=Control, loc=<0 0 0>, rot=<0 0 0 1>}
 FK_armRot_r_ctrl {controlType=EULER_TRANSFORM, parent=bp_pelvis_ctrl, type=Control, loc=<0 0 0>, rot=<0 0 0 1>}
 FK_upperarm_r_ctrl {controlType=EULER_TRANSFORM, parent=FK_armRot_r_ctrl, type=Control, loc=<0 0 0>, rot=<0 0 0 1>}
 FK_lowerarm_r_ctrl {controlType=EULER_TRANSFORM, parent=FK_upperarm_r_ctrl, type=Control, loc=<0 0 0>, rot=<0 0 0 1>}
 FK_hand_r_ctrl {controlType=EULER_TRANSFORM, parent=FK_lowerarm_r_ctrl, type=Control, loc=<0 0 0>, rot=<0 0 0 1>}
 IK_foot_l_ctrl {controlType=EULER_TRANSFORM, parent=Main_ctrl, type=Control, loc=<0 0 0>, rot=<0 0 0 1>}
 IK_calf_l_ctrl {controlType=EULER_TRANSFORM, parent=IK_foot_l_ctrl, type=Control, loc=<0 0 0>, rot=<0 0 0 1>}
 IK_Heel_RV_l_ctrl {controlType=EULER_TRANSFORM, parent=IK_foot_l_ctrl, type=Control, loc=<0 0 0>, rot=<0 0 0 1>}
 IK_Toe_RV_l_ctrl {controlType=EULER_TRANSFORM, parent=IK_Heel_RV_l_ctrl, type=Control, loc=<0 0 0>, rot=<0 0 0 1>}
 IK_Ball_RV_l_ctrl {controlType=EULER_TRANSFORM, parent=IK_Toe_RV_l_ctrl, type=Control, loc=<0 0 0>, rot=<0 0 0 1>}
 IK_Foot_RV_l {parent=IK_Ball_RV_l_ctrl, type=Null, loc=<0 0 0>, rot=<0 0 0 1>}
 IK_Ball_RV_l {parent=IK_Toe_RV_l_ctrl, type=Null, loc=<-5.3498 -7.7029 -19.6606>, rot=<-0.037 0.0681 -0.8196 0.5676>}
 IK_foot_r_ctrl {controlType=EULER_TRANSFORM, parent=Main_ctrl, type=Control, loc=<0 0 0>, rot=<0 0 0 1>}
 IK_calf_r_ctrl {controlType=EULER_TRANSFORM, parent=IK_foot_r_ctrl, type=Control, loc=<0 0 0>, rot=<0 0 0 1>}
 IK_Heel_RV_r_ctrl {controlType=EULER_TRANSFORM, parent=IK_foot_r_ctrl, type=Control, loc=<0 0 0>, rot=<0 0 0 1>}
 IK_Toe_RV_r_ctrl {controlType=EULER_TRANSFORM, parent=IK_Heel_RV_r_ctrl, type=Control, loc=<0 0 0>, rot=<0 0 0 1>}
 IK_Ball_RV_r_ctrl {controlType=EULER_TRANSFORM, parent=IK_Toe_RV_r_ctrl, type=Control, loc=<0 0 0>, rot=<0 0 0 1>}
 IK_Foot_RV_r {parent=IK_Ball_RV_r_ctrl, type=Null, loc=<0 0 0>, rot=<0 0 0 1>}
 IK_Ball_RV_r {parent=IK_Toe_RV_r_ctrl, type=Null, loc=<0 0 0>, rot=<0 0 0 1>}
 IK_calf_l_null_ref {parent=Main_ctrl, type=Null, loc=<-7.5238 -8.8977 7.5539>, rot=<0.5806 0.4208 -0.6209 0.3169>}
 IK_calf_r_null_ref {parent=Main_ctrl, type=Null, loc=<7.5238 -8.8977 7.5539>, rot=<0.4208 0.5806 -0.3169 0.6209>}
 bp_tie_ctrl {controlType=EULER_TRANSFORM, parent=Main_ctrl, type=Control, loc=<0 0 0>, rot=<0 0 0 1>}
]
