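# Animation Blueprint for the ScoopPluck backpack that plays a reactive animation when the player eliminates an enemy.
# A state machine moves from a frozen Idle pose to OnKill on activeKill, set by the elimination anim event or front-end timer and delay logic.
# An additive Modify Bone then offsets the pack bone using owner spine yaw.
# The Control Rig and reactivity blend nodes are not connected to the output pose.
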AnimBlueprint /BRCosmetics/Accessories/FORT_Backpacks/Backpack_M_MED_ScoopPluck/Meshes/M_MED_ScoopPluck_Pack_AnimBP
target_skeleton=&/M_MED_ScoopPluck_Pack_Skeleton
graphs [
 AnimGraph {nodes [
   Blend Poses by bool(BlendPose_1=Sequence Player(), bActiveValue=Get CosmeticVariantReactivity() -> (CosmeticVariantReactivity:bool)) {blendType=Linear}
   Control Rig(Source=Input Pose()) {controlRigAssetReference {blueprintRigClass=&/M_MED_ScoopPluck_Pack_CtrlRig.M_MED_ScoopPluck_Pack_CtrlRig_C}}
   AnimGraphNode_StateMachine_0 = OnKillStateMachine() {is=title AnimGraphNode_StateMachine}
   n1 = "Transform (Modify) Bone - Bone: pack"(
    ComponentPose=Local To Component(LocalPose=@AnimGraphNode_StateMachine_0)
    Translation=Get backpack_translation()
    Rotation=Get backpack_rotate()
   ) {
    rotationMode=BMM_Additive
    rotationSpace=BCS_BoneSpace
    translationMode=BMM_Additive
    translationSpace=BCS_BoneSpace
    alphaScaleBiasClamp {bInterpResult=true, interpSpeedDecreasing=5, interpSpeedIncreasing=5}
    boneToModify {boneName=pack}
   }
   Output Pose(Result=Component To Local(ComponentPose=@n1)) {layerGroup=None}
  ]}
 OnKillStateMachine {
  entry=Idle
  path=AnimGraph.AnimGraphNode_StateMachine_0.OnKillStateMachine
  ids=id
  nodes [
   AnimStateNode_2 = Idle()
   AnimStateNode_3 = OnKill() {is=title AnimStateNode}
   AnimStateTransitionNode_2 = AnimStateTransitionNode() {bAllowInertializationForSelfTransitions=true}
   AnimStateTransitionNode_3 = AnimStateTransitionNode() {bAllowInertializationForSelfTransitions=true, bAutomaticRuleBasedOnSequencePlayerInState=true}
  ]
  links [
   {
    dst=OnKill
    rule=AnimGraph.AnimGraphNode_StateMachine_0.OnKillStateMachine.AnimStateTransitionNode_2.Transition
    src=Idle
    via=AnimStateTransitionNode_2
   }
   {
    dst=Idle
    rule=AnimGraph.AnimGraphNode_StateMachine_0.OnKillStateMachine.AnimStateTransitionNode_3.Transition
    src=OnKill
    via=AnimStateTransitionNode_3
   }
  ]
 }
 Idle {
  path=AnimGraph.AnimGraphNode_StateMachine_0.OnKillStateMachine.AnimStateNode_2.Idle
  nodes [
   n1 = Sequence Player 'M_MED_ScoopPluck_Reactive_PACK'() {
    bLoopAnimation=false
    playRate=0
    sequence=/BRCosmetics/Accessories/FORT_Backpacks/Backpack_M_MED_ScoopPluck/Anim/M_MED_ScoopPluck_Reactive_PACK
    playRateScaleBiasClampConstants {clampMax=0}
   }
   Output Animation Pose(Result=@n1)
  ]
 }
 OnKill {
  path=AnimGraph.AnimGraphNode_StateMachine_0.OnKillStateMachine.AnimStateNode_3.OnKill
  nodes [
   n1 = Sequence Player 'M_MED_ScoopPluck_Reactive_PACK'() {bLoopAnimation=false, sequence=&M_MED_ScoopPluck_Reactive_PACK}
   Output Animation Pose(Result=@n1)
  ]
 }
 Transition {
  path=AnimGraph.AnimGraphNode_StateMachine_0.OnKillStateMachine.AnimStateTransitionNode_2.Transition
  nodes [Result(bCanEnterTransition=Get activeKill() -> (activeKill:bool))]
 }
 Transition {path=AnimGraph.AnimGraphNode_StateMachine_0.OnKillStateMachine.AnimStateTransitionNode_3.Transition, nodes [Result()]}
 EventGraph {nodes [
   n1 = Timer Exec() {is=title K2Node_CustomEvent}
   > Branch(Condition=Get CosmeticVariantReactivity())
   > Set activeKill(activeKill:bool=true) -> (Output_Get:bool)
   n2 = Get KillCounter() -> (KillCounter:int=0)
   > Set KillCounter(KillCounter:int="Clamp (Integer)"(Value=int + int(A=@n2), Min=0, Max=0) -> (ReturnValue=0)) -> (Output_Get:int=0)
   Event BlueprintUpdateAnimation()
   > n3 = Sequence()
   n4 = GetOwnerSkeletalMeshComponent()
   n5 = BreakRotator(InRot=BreakTransform(InTransform=GetDeltaTransformFromRefPose(self=@n4, BoneName=spine_05, BaseName=spine_01)).Rotation)
   n6 = MapRangeClamped(Value=@n5.Yaw, InRangeA=5, InRangeB=-20, OutRangeA=-5, OutRangeB=15)
   > Set backpack_rotate(backpack_rotate=MakeRotator(Roll=FInterpEaseInOut!(A:double=0, B:double=@n6, Alpha:double=0, Exponent:double=0) -> (ReturnValue:double)))
   n7 = MapRangeClamped(Value=@n5.Yaw, InRangeA=0, InRangeB=-10, OutRangeB=-1.5)
   > Set backpack_translation(backpack_translation=MakeVector(Y=FInterpEaseInOut(A=0, B=@n7, Alpha=0, Exponent=0)))
   n8 = Sequence!(execute=@n3.then_1) -> (then_0, then_1, then_2:Exec)
   n9 = Get bodyMesh() -> (bodyMesh:Skeletal Mesh Component Object Reference)
   > Is Valid(InputObject=@n9)
   > Set PelvisLocation(PelvisLocation:Vector=GetSocketLocation!(self:Scene Component Object Reference=@n9, InSocketName:Name=pelvis) -> (ReturnValue:Vector)) -> (Output_Get:Vector)
   > Branch(Condition=Get bIsInFrontEnd())
   > Branch(Condition="integer >= integer"(A=Get KillCounter() -> (KillCounter=0)))
   > Do Once(Reset=@n13.then)
   > n10 = Delay(Duration=40)
   Do Once(execute=@n14.then, Reset=@n10.then)
   > Sequence!(execute) -> (then_0, then_1)
   > Set activeKill(activeKill=false)
   > Set KillCounter(KillCounter=0) -> (Output_Get=0)
   > n11 = Set Timer by Event(Delegate=@n1, Time=Get Time() -> (Time:float), bLooping=true)
   > n12 = Set TimerHandle(TimerHandle:Timer Handle Structure=@n11) -> (Output_Get:Timer Handle Structure)
   > n13 = Unpause Timer by Handle!(execute, Handle:Timer Handle Structure=@n12) -> (then)
   n14 = Branch(execute=@n8.then_1, Condition=Get bIsInFrontEnd())
   Branch(execute=@n8.then_2, Condition=Get bIsInFrontEnd())
   > Do Once()
   > n15 = Branch(Condition="integer > integer"(A=Get KillCounter() -> (KillCounter=0), B=3))
   > n16 = Delay(Duration=10)
   RetriggerableDelay(execute=[@n16.then, @n18.then, @n26.then, @n28.then], Duration=0.1)
   > Set activeKill(activeKill=true)
   > n17 = Increment Int(Value=Get KillCounter() -> (KillCounter=0))
   > Set KillCounter(KillCounter=@n17) -> (Output_Get=0)
   n18 = Delay(execute=@n15.else, Duration=1.5)
   n19 = Event BlueprintPostEvaluateAnimation() {is=title K2Node_Event}
   Set activeKill(execute=[@n19.then, @n27.then], activeKill=false)
   Event OnCharacterCustomizationLoaded()
   > n20 = Sequence()
   > Set SkeletalMesh(SkeletalMesh:Skeletal Mesh Component Object Reference=GetPartSkeletalMeshComponent()) -> (Output_Get:Skeletal Mesh Component Object Reference)
   > n21 = Cast To FortPlayerPawnAthena(Object=TryGetPawnOwner())
   > Set bodyMesh(bodyMesh:Skeletal Mesh Component Object Reference=GetSkeletalMeshForPartType(self=@n21)) -> (Output_Get:Skeletal Mesh Component Object Reference)
   n22 = Cast To FortPlayerPawn(execute=@n20.then_1, Object=TryGetFortPawn())
   > n23 = Cast To B_Backpack_LastVoiceDive!(
    execute
    Object=GetModifierForPartType!(self:Fort Player Pawn Object Reference=@n22, PartType:EFortCustomPartType Enum=Backpack) -> (ReturnValue:Custom Character Part Modifier Object Reference)
   ) -> (then, CastFailed, AsB Backpack Last Voice Dive:B Backpack Last Voice Dive Object Reference) {is=title K2Node_DynamicCast}
   > n24 = Set As B Backpack Last Voice Dive(As B Backpack Last Voice Dive:B Backpack Last Voice Dive Object Reference=@n23) -> (Output_Get:B Backpack Last Voice Dive Object Reference)
   > Set CosmeticVariantReactivity(CosmeticVariantReactivity:bool=Get reactivityEnabled!(self:B Backpack Last Voice Dive Object Reference=@n24) -> (reactivityEnabled:bool)) -> (Output_Get:bool)
   n25 = Event BlueprintProcessAnimInputEvent()
   > Sequence()
   > n26 = Branch(Condition=IsNamed(self=@n25.AnimInputEvent, InEventName=aie.EnemyPlayerEliminated))
   n27 = Branch(Condition=Get CosmeticVariantReactivity())
   n28 = Branch(Condition=Get CosmeticVariantReactivity())
  ]}
]
variables [
 activeKill {type=boolean}
 As B Backpack Last Voice Dive {type=B_Backpack_LastVoiceDive_C, default=null}
 backpack_rotate {type=Rotator}
 backpack_translation {type=Vector}
 bodyMesh {type=SkeletalMeshComponent}
 CosmeticVariantReactivity {type=boolean}
 KillCounter {type=integer}
 PelvisLocation {type=Vector}
 SkeletalMesh {type=SkeletalMeshComponent}
 Time {type=number}
 TimerHandle {type=TimerHandle}
]
